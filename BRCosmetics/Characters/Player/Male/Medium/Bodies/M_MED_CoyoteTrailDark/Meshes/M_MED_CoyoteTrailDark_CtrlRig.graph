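# Control Rig for the M_MED_CoyoteTrailDark player body that procedurally adjusts its strap, collar and right clavicle deform bones.
# Spherical pose readers on thighs, head, right clavicle and left upper arm drive corrective weights, including collar tilt while aiming and helmet-dependent collar offsets.
# Interpolated movement, crouch and jump inputs weight additional front and back strap offsets.
# Inner front straps follow a weighted parent constraint between both outer straps and the pelvis.
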ControlRigBlueprint /BRCosmetics/Characters/Player/Male/Medium/Bodies/M_MED_CoyoteTrailDark/Meshes/M_MED_CoyoteTrailDark_CtrlRig
graphs [
 ContainedGraph {nodes [
   n1 = RigVMFunctionEntryNode()
   SequenceAggregate_1_1 = &RigVMFunction_Sequence(ExecuteContext=@n1.ExecuteContext)
   SequenceAggregate_1_1_1 = &RigVMFunction_Sequence(ExecuteContext=@SequenceAggregate_1_1.B)
   SequenceAggregate_1_1_2 = &RigVMFunction_Sequence(ExecuteContext=@SequenceAggregate_1_1_1.B)
   SequenceAggregate_1_1_4 = &RigVMFunction_Sequence(ExecuteContext=@SequenceAggregate_1_1_2.B)
   SequenceAggregate_1_1_3 = &RigVMFunction_Sequence(ExecuteContext=@SequenceAggregate_1_1_4.B)
   RigVMFunctionReturnNode(
    A=@SequenceAggregate_1_1.A
    B=@SequenceAggregate_1_1_1.A
    C=@SequenceAggregate_1_1_2.A
    F=@SequenceAggregate_1_1_3.A
    G=@SequenceAggregate_1_1_3.B
    E=@SequenceAggregate_1_1_4.A
   )
  ]}
 ContainedGraph {nodes [
   n1 = RigVMFunctionEntryNode()
   SequenceAggregate = &RigVMFunction_Sequence(ExecuteContext=@n1.ExecuteContext)
   SequenceAggregate_1 = &RigVMFunction_Sequence(ExecuteContext=@SequenceAggregate.B)
   SequenceAggregate_2 = &RigVMFunction_Sequence(ExecuteContext=@SequenceAggregate_1.B)
   SequenceAggregate_3 = &RigVMFunction_Sequence(ExecuteContext=@SequenceAggregate_2.B)
   SequenceAggregate_4 = &RigVMFunction_Sequence(ExecuteContext=@SequenceAggregate_3.B)
   SequenceAggregate_5 = &RigVMFunction_Sequence(ExecuteContext=@SequenceAggregate_4.B)
   SequenceAggregate_6 = &RigVMFunction_Sequence(ExecuteContext=@SequenceAggregate_5.B)
   RigVMFunctionReturnNode(
    A=@SequenceAggregate.A
    B=@SequenceAggregate_1.A
    C=@SequenceAggregate_2.A
    D=@SequenceAggregate_3.A
    E=@SequenceAggregate_4.A
    F=@SequenceAggregate_5.A
    G=@SequenceAggregate_6.A
    H=@SequenceAggregate_6.B
   )
  ]}
 ContainedGraph {nodes [
   n1 = RigVMFunctionEntryNode()
   SequenceAggregate_6 = &RigVMFunction_Sequence(ExecuteContext=@n1.ExecuteContext)
   SequenceAggregate_6_1 = &RigVMFunction_Sequence(ExecuteContext=@SequenceAggregate_6.B)
   SequenceAggregate_6_2 = &RigVMFunction_Sequence(ExecuteContext=@SequenceAggregate_6_1.B)
   SequenceAggregate_6_4 = &RigVMFunction_Sequence(ExecuteContext=@SequenceAggregate_6_2.B)
   RigVMFunctionReturnNode(A=@SequenceAggregate_6.A, B=@SequenceAggregate_6_1.A, C=@SequenceAggregate_6_2.A, E=@SequenceAggregate_6_4.A, F=@SequenceAggregate_6_4.B)
  ]}
 ContainedGraph {nodes [
   n1 = RigVMFunctionEntryNode()
   SequenceAggregate_5 = &RigVMFunction_Sequence(ExecuteContext=@n1.ExecuteContext)
   SequenceAggregate_5_1 = &RigVMFunction_Sequence(ExecuteContext=@SequenceAggregate_5.B)
   SequenceAggregate_5_2 = &RigVMFunction_Sequence(ExecuteContext=@SequenceAggregate_5_1.B)
   RigVMFunctionReturnNode(A=@SequenceAggregate_5.A, B=@SequenceAggregate_5_1.A, C=@SequenceAggregate_5_2.A, D=@SequenceAggregate_5_2.B)
  ]}
 ContainedGraph {nodes [
   n1 = RigVMFunctionEntryNode()
   SequenceAggregate_4 = &RigVMFunction_Sequence(ExecuteContext=@n1.ExecuteContext)
   SequenceAggregate_4_1 = &RigVMFunction_Sequence(ExecuteContext=@SequenceAggregate_4.B)
   SequenceAggregate_4_2 = &RigVMFunction_Sequence(ExecuteContext=@SequenceAggregate_4_1.B)
   RigVMFunctionReturnNode(A=@SequenceAggregate_4.A, B=@SequenceAggregate_4_1.A, C=@SequenceAggregate_4_2.A, D=@SequenceAggregate_4_2.B)
  ]}
 ContainedGraph {nodes [
   n1 = RigVMFunctionEntryNode()
   SequenceAggregate_1 = &RigVMFunction_Sequence(ExecuteContext=@n1.ExecuteContext)
   SequenceAggregate_1_1 = &RigVMFunction_Sequence(ExecuteContext=@SequenceAggregate_1.B)
   SequenceAggregate_1_2 = &RigVMFunction_Sequence(ExecuteContext=@SequenceAggregate_1_1.B)
   SequenceAggregate_1_3 = &RigVMFunction_Sequence(ExecuteContext=@SequenceAggregate_1_2.B)
   RigVMFunctionReturnNode(A=@SequenceAggregate_1.A, B=@SequenceAggregate_1_1.A, C=@SequenceAggregate_1_2.A, D=@SequenceAggregate_1_3.A, E=@SequenceAggregate_1_3.B)
  ]}
 ContainedGraph {nodes [
   n1 = RigVMFunctionEntryNode()
   SequenceAggregate_2 = &RigVMFunction_Sequence(ExecuteContext=@n1.ExecuteContext)
   SequenceAggregate_2_1 = &RigVMFunction_Sequence(ExecuteContext=@SequenceAggregate_2.B)
   RigVMFunctionReturnNode(A=@SequenceAggregate_2.A, B=@SequenceAggregate_2_1.A, C=@SequenceAggregate_2_1.B)
  ]}
 RigVMFunctionLibrary {links [], nodes []}
 RigVMModel {nodes [
   n1 = &RigVMFunction_MathQuaternionFromEuler(Euler=(X=0,Y=0,Z=-45))
   &RigUnit_ModifyTransforms(
    ExecutePin=()
    ItemToModify=((Item=(Type=Bone,Name="dfrm_front_strap_inner_mstr_l"),Transform=(Rotation=(X=0,Y=0,Z=0,W=1),Translation=(X=0,Y=0,Z=0),Scale3D=(X=1,Y=1,Z=1))))
    ItemToModify.0.Transform.Rotation=@n1.Result
    Weight=RigVMVariableNode() {variable=backward_interp_alpha}.Value
   )
   > &RigUnit_ModifyTransforms(
    ItemToModify=((Item=(Type=Bone,Name="dfrm_front_strap_inner_mstr_r"),Transform=(Rotation=(X=0,Y=0,Z=0,W=1),Translation=(X=0,Y=0,Z=0),Scale3D=(X=1,Y=1,Z=1))))
    Weight=RigVMVariableNode() {variable=backward_interp_alpha}.Value
    ItemToModify.0.Transform.Rotation=@n1.Result
   )
   RigUnit_BeginExecution = &RigUnit_BeginExecution()
   SequenceAggregate_1_1 = RigVMAggregateNode(ExecuteContext=@RigUnit_BeginExecution.ExecutePin)
   SequenceAggregate_3 = &RigVMFunction_Sequence(ExecuteContext=@SequenceAggregate_1_1.B)
   n2 = RigVMVariableNode() {variable=forward_backwards_movement_alpha}
   RigVMVariableNode(
    Value=&RigVMFunction_MathDoubleRemap(SourceMaximum=1, SourceMinimum=0, TargetMaximum=1, TargetMinimum=0, Value=@n2.Value).Result
    ExecuteContext=@SequenceAggregate_3.A
   ) {variable=forward_alpha}
   Remap = &RigVMFunction_MathDoubleRemap(SourceMaximum=0, SourceMinimum=-1, TargetMaximum=0, TargetMinimum=1, Value=@n2.Value)
   > RigVMVariableNode(Value=@Remap.Result) {variable=backwards_alpha}
   n3 = RigVMVariableNode() {variable=left_right_movement_alpha}
   RigVMVariableNode(
    Value=&RigVMFunction_MathDoubleRemap(SourceMaximum=0, SourceMinimum=-1, TargetMaximum=0, TargetMinimum=1, Value=@n3.Value).Result
    ExecuteContext=@SequenceAggregate_3.B
   ) {variable=left_alpha}
   > RigVMVariableNode(Value=&RigVMFunction_MathDoubleRemap(SourceMaximum=1, SourceMinimum=0, TargetMaximum=1, TargetMinimum=0, Value=@n3.Value).Result) {variable=right_alpha}
   &RigUnit_ModifyTransforms(
    ItemToModify=((Item=(Type=Bone,Name="dfrm_clavicle_r"),Transform=(Rotation=(X=0,Y=0,Z=0,W=1),Translation=(X=0,Y=0,Z=0),Scale3D=(X=0.95,Y=1,Z=1))))
    ItemToModify.0.Transform.Rotation=&RigVMFunction_MathQuaternionFromEuler(Euler=(X=-30,Y=0,Z=0)).Result
    Weight=RigVMVariableNode() {variable=spr_clavicle_r_twist}.Value
    ExecutePin=@SequenceAggregate_1_1.F
   )
   &RigUnit_ModifyTransforms(
    ItemToModify=((Item=(Type=Bone,Name="dfrm_collar_mstr"),Transform=(Rotation=(X=0,Y=0,Z=0,W=1),Translation=(X=0,Y=0,Z=0),Scale3D=(X=1,Y=1,Z=1))))
    ItemToModify.0.Transform.Rotation=&RigVMFunction_MathQuaternionFromEuler(Euler=(X=0,Y=-10,Z=0)).Result
    Weight=RigVMVariableNode() {variable=head_tilt_aim_alpha}.Value
    ExecutePin=@SequenceAggregate_1_1.E
   )
   > &RigUnit_ModifyTransforms(
    ItemToModify=((Item=(Type=Bone,Name="dfrm_collar_back_l"),Transform=(Rotation=(X=0,Y=0,Z=0,W=1),Translation=(X=0,Y=-2,Z=0),Scale3D=(X=1,Y=1,Z=1))))
    ItemToModify.0.Transform.Rotation=&RigVMFunction_MathQuaternionFromEuler(Euler=(X=5,Y=25,Z=0)).Result
    Weight=RigVMVariableNode() {variable=head_tilt_aim_alpha}.Value
   )
   > &RigUnit_ModifyTransforms(
    ItemToModify=((Item=(Type=Bone,Name="dfrm_collar_back"),Transform=(Rotation=(X=0,Y=0,Z=0,W=1),Translation=(X=0,Y=0,Z=0),Scale3D=(X=1,Y=1,Z=1))))
    ItemToModify.0.Transform.Rotation=&RigVMFunction_MathQuaternionFromEuler(Euler=(X=15,Y=0,Z=0)).Result
    Weight=RigVMVariableNode() {variable=head_tilt_aim_alpha}.Value
   )
   > &RigUnit_ModifyTransforms(
    ItemToModify=((Item=(Type=Bone,Name="dfrm_collar_back_r"),Transform=(Rotation=(X=0,Y=0,Z=0,W=1),Translation=(X=0,Y=1,Z=0),Scale3D=(X=1,Y=1,Z=1))))
    Weight=RigVMVariableNode() {variable=head_tilt_aim_alpha}.Value
   )
   SequenceAggregate = RigVMAggregateNode(ExecuteContext=@SequenceAggregate_1_1.A)
   n4 = &RigUnit_SphericalPoseReader(
    ActiveRegionScaleFactors=(PositiveWidth=3.5,NegativeWidth=0.5,PositiveHeight=3,NegativeHeight=3)
    ActiveRegionSize=0.1
    Debug=(bDrawDebug=false,DebugScale=25,DebugSegments=20,DebugThickness=0.25)
    DriverAxis=(X=-1,Y=0,Z=0)
    DriverItem=(Type=Bone,Name="thigh_l")
    FalloffRegionScaleFactors=(PositiveWidth=1,NegativeWidth=1,PositiveHeight=1,NegativeHeight=1)
    FalloffSize=0.45
    OptionalParentItem=(Type=Bone)
    RotationOffset=(X=90,Y=0,Z=0)
    ExecutePin=@SequenceAggregate.A
   )
   > RigVMVariableNode(Value=@n4.OutputParam) {variable=spr_thigh_l_up_alpha}
   n5 = &RigUnit_SphericalPoseReader(
    ActiveRegionScaleFactors=(PositiveWidth=3.5,NegativeWidth=0.5,PositiveHeight=3,NegativeHeight=3)
    ActiveRegionSize=0.1
    Debug=(bDrawDebug=false,DebugScale=25,DebugSegments=20,DebugThickness=0.25)
    DriverAxis=(X=1,Y=0,Z=0)
    DriverItem=(Type=Bone,Name="thigh_r")
    FalloffRegionScaleFactors=(PositiveWidth=1,NegativeWidth=1,PositiveHeight=1,NegativeHeight=1)
    FalloffSize=0.45
    FlipHeightScaling=false
    FlipWidthScaling=true
    OptionalParentItem=(Type=Bone)
    RotationOffset=(X=-90,Y=0,Z=0)
    ExecutePin=@SequenceAggregate.B
   )
   > RigVMVariableNode(Value=@n5.OutputParam) {variable=spr_thigh_r_up_alpha}
   n6 = &RigUnit_SphericalPoseReader(
    ActiveRegionScaleFactors=(PositiveWidth=2,NegativeWidth=0.6,PositiveHeight=2,NegativeHeight=2)
    ActiveRegionSize=0.125
    Debug=(bDrawDebug=false,DebugScale=25,DebugSegments=20,DebugThickness=0.25)
    DriverAxis=(X=0,Y=1,Z=0)
    DriverItem=(Type=Bone,Name="head")
    FalloffRegionScaleFactors=(PositiveWidth=1,NegativeWidth=0.825,PositiveHeight=1,NegativeHeight=1)
    FalloffSize=0.25
    OptionalParentItem=(Type=Bone,Name="None")
    RotationOffset=(X=90,Y=0,Z=-40)
    ExecutePin=@SequenceAggregate.C
   )
   > RigVMVariableNode(Value=@n6.OutputParam) {variable=spr_head_alpha}
   n7 = &RigUnit_SphericalPoseReader(
    ActiveRegionScaleFactors=(PositiveWidth=1,NegativeWidth=1,PositiveHeight=1,NegativeHeight=1)
    ActiveRegionSize=0.05
    Debug=(bDrawDebug=false,DebugScale=25,DebugSegments=20,DebugThickness=0.25)
    DriverAxis=(X=1,Y=0,Z=0)
    DriverItem=(Type=Bone,Name="head")
    FalloffRegionScaleFactors=(PositiveWidth=1,NegativeWidth=1,PositiveHeight=1,NegativeHeight=1)
    FalloffSize=0.1
    OptionalParentItem=(Type=Bone)
    RotationOffset=(X=-30,Y=-75,Z=0)
    ExecutePin=@SequenceAggregate.D
   )
   > RigVMVariableNode(Value=@n7.OutputParam) {variable=spr_head_tilt_alpha}
   n8 = &RigUnit_SphericalPoseReader(
    ActiveRegionScaleFactors=(PositiveWidth=1,NegativeWidth=1,PositiveHeight=1,NegativeHeight=1)
    ActiveRegionSize=0.1
    Debug=(bDrawDebug=false,DebugScale=25,DebugSegments=20,DebugThickness=0.25)
    DriverAxis=(X=0,Y=0,Z=-1)
    DriverItem=(Type=Bone,Name="clavicle_r")
    FalloffRegionScaleFactors=(PositiveWidth=1,NegativeWidth=1,PositiveHeight=1,NegativeHeight=1)
    FalloffSize=0.25
    OptionalParentItem=(Type=Bone)
    RotationOffset=(X=135,Y=0,Z=0)
    ExecutePin=@SequenceAggregate.E
   )
   > RigVMVariableNode(Value=@n8.OutputParam) {variable=spr_clavicle_r_twist}
   n9 = &RigUnit_SphericalPoseReader(
    ActiveRegionScaleFactors=(PositiveWidth=1,NegativeWidth=1,PositiveHeight=1,NegativeHeight=1)
    ActiveRegionSize=0.1
    Debug=(bDrawDebug=false,DebugScale=25,DebugSegments=20,DebugThickness=0.25)
    DriverAxis=(X=1,Y=0,Z=0)
    DriverItem=(Type=Bone,Name="upperarm_l")
    FalloffRegionScaleFactors=(PositiveWidth=1,NegativeWidth=1,PositiveHeight=1,NegativeHeight=1)
    FalloffSize=0.2
    OptionalParentItem=(Type=Bone)
    RotationOffset=(X=70,Y=0,Z=-25)
    ExecutePin=@SequenceAggregate.F
   )
   > RigVMVariableNode(Value=@n9.OutputParam) {variable=spr_upperarm_l_aim_pistol_alpha}
   n10 = &RigUnit_SphericalPoseReader(
    ActiveRegionScaleFactors=(PositiveWidth=1,NegativeWidth=1,PositiveHeight=1,NegativeHeight=1)
    ActiveRegionSize=0.1
    Debug=(bDrawDebug=false,DebugScale=25,DebugSegments=20,DebugThickness=0.25)
    DriverAxis=(X=1,Y=0,Z=0)
    DriverItem=(Type=Bone,Name="upperarm_l")
    FalloffRegionScaleFactors=(PositiveWidth=1,NegativeWidth=1,PositiveHeight=1,NegativeHeight=1)
    FalloffSize=0.2
    OptionalParentItem=(Type=Bone)
    RotationOffset=(X=65,Y=0,Z=-60)
    ExecutePin=@SequenceAggregate.G
   )
   > RigVMVariableNode(Value=@n10.OutputParam) {variable=spr_upperarm_l_aim_rifle_alpha}
   n11 = RigVMVariableNode() {variable=spr_head_tilt_alpha}
   Maximum = &RigVMFunction_MathDoubleMax(
    A=&RigVMFunction_MathDoubleMul(A=@n11.Value, B=RigVMVariableNode() {variable=spr_upperarm_l_aim_pistol_alpha}.Value).Result
    B=&RigVMFunction_MathDoubleMul(B=RigVMVariableNode() {variable=spr_upperarm_l_aim_rifle_alpha}.Value, A=@n11.Value).Result
   )
   n12 = &RigVMFunction_AlphaInterp(InterpSpeedDecreasing=10, InterpSpeedIncreasing=10, Scale=1, bClampResult=False, Value=@Maximum.Result)
   RigVMVariableNode(Value=@n12.Result, ExecuteContext=@SequenceAggregate.H) {variable=head_tilt_aim_alpha}
   SequenceAggregate_1 = RigVMAggregateNode(ExecuteContext=@SequenceAggregate_1_1.G)
   n13 = RigVMVariableNode() {variable=is_crouching}
   Multiply = &RigVMFunction_MathDoubleMul(
    A=&RigVMFunction_MathDoubleSub(A=1, B=&RigVMFunction_MathBoolToFloat(Value=@n13.Value).Result).Result
    B=RigVMVariableNode() {variable=forward_alpha}.Value
   )
   n14 = &RigVMFunction_AlphaInterp(ClampMax=0.1, ClampMin=0.1, InterpSpeedDecreasing=5, InterpSpeedIncreasing=1, Scale=1, bClampResult=False, Value=@Multiply.Result)
   &RigUnit_ModifyTransforms(
    ItemToModify=((Item=(Type=Bone,Name="dfrm_front_strap_outer_mstr_l"),Transform=(Rotation=(X=0,Y=0,Z=0,W=1),Translation=(X=-3,Y=-0.75,Z=-0.5),Scale3D=(X=1,Y=1,Z=1))))
    ItemToModify.0.Transform.Rotation=&RigVMFunction_MathQuaternionFromEuler(Euler=(X=0,Y=5,Z=-10)).Result
    ExecutePin=@SequenceAggregate_1.A
    Weight=@n14.Result
   )
   > &RigUnit_ModifyTransforms(
    ItemToModify=((Item=(Type=Bone,Name="dfrm_front_strap_outer_mstr_r"),Transform=(Rotation=(X=0,Y=0,Z=0,W=1),Translation=(X=-3,Y=-0.75,Z=0.5),Scale3D=(X=1,Y=1,Z=1))))
    ItemToModify.0.Transform.Rotation=&RigVMFunction_MathQuaternionFromEuler(Euler=(X=0,Y=-5,Z=-10)).Result
    Weight=@n14.Result
   )
   SequenceAggregate_5 = RigVMAggregateNode(ExecuteContext=@SequenceAggregate_1.C)
   n15 = &RigVMFunction_AlphaInterp(
    InterpSpeedDecreasing=10
    InterpSpeedIncreasing=10
    Scale=1
    bClampResult=False
    bMapRange=true
    Value=&RigVMFunction_MathBoolToFloat(Value=RigVMVariableNode() {variable=is_crouching}.Value).Result
   )
   n16 = &RigVMFunction_AlphaInterp(InterpSpeedDecreasing=5, InterpSpeedIncreasing=5, Scale=1, bClampResult=False, Value=RigVMVariableNode() {variable=forward_alpha}.Value)
   RigVMVariableNode(Value=&RigVMFunction_MathDoubleMul(A=@n15.Result, B=@n16.Result).Result, ExecuteContext=@SequenceAggregate_5.A) {variable=forward_interp_alpha}
   n17 = &RigVMFunction_AlphaInterp(InterpSpeedDecreasing=5, InterpSpeedIncreasing=5, Scale=1, bClampResult=False, Value=RigVMVariableNode() {variable=backwards_alpha}.Value)
   > RigVMVariableNode(Value=&RigVMFunction_MathDoubleMul(A=@n15.Result, B=@n17.Result).Result) {variable=backward_interp_alpha}
   n18 = &RigVMFunction_AlphaInterp(InterpSpeedDecreasing=5, InterpSpeedIncreasing=5, Scale=1, bClampResult=False, Value=RigVMVariableNode() {variable=left_alpha}.Value)
   n19 = &RigVMFunction_MathDoubleMul(B=@n18.Result, A=@n15.Result)
   > RigVMVariableNode(Value=@n19.Result) {variable=left_interp_alpha}
   n20 = &RigVMFunction_AlphaInterp(InterpSpeedDecreasing=5, InterpSpeedIncreasing=5, Scale=1, bClampResult=False, Value=RigVMVariableNode() {variable=right_alpha}.Value)
   n21 = &RigVMFunction_MathDoubleMul(B=@n20.Result, A=@n15.Result)
   > RigVMVariableNode(Value=&RigVMFunction_MathDoubleMax(B=@n21.Result, A=@n19.Result).Result) {variable=left_right_interp_alpha}
   > RigVMVariableNode(Value=@n21.Result) {variable=right_interp_alpha}
   Subtract = &RigVMFunction_MathDoubleSub(A=1, B=RigVMVariableNode() {variable=forward_interp_alpha}.Value)
   RigVMVariableNode(
    Value=&RigVMFunction_MathDoubleMul(A=@Subtract.Result, B=RigVMVariableNode() {variable=left_right_interp_alpha}.Value).Result
    ExecuteContext=@SequenceAggregate_5.B
   ) {variable=left_right_interp_alpha}
   n22 = RigVMVariableNode() {variable=forward_interp_alpha}
   RigVMVariableNode(
    Value=&RigVMFunction_MathDoubleMul(A=&RigVMFunction_MathDoubleSub(A=1, B=@n22.Value).Result, B=RigVMVariableNode() {variable=left_interp_alpha}.Value).Result
    ExecuteContext=@SequenceAggregate_5.C
   ) {variable=left_interp_alpha}
   n23 = RigVMVariableNode() {variable=forward_interp_alpha}
   RigVMVariableNode(
    Value=&RigVMFunction_MathDoubleMul(A=&RigVMFunction_MathDoubleSub(A=1, B=@n23.Value).Result, B=RigVMVariableNode() {variable=right_interp_alpha}.Value).Result
    ExecuteContext=@SequenceAggregate_5.D
   ) {variable=right_interp_alpha}
   SequenceAggregate_4 = RigVMAggregateNode(ExecuteContext=@SequenceAggregate_1.E)
   Cast = &RigVMFunction_MathBoolToFloat(Value=RigVMVariableNode() {variable=is_jumping}.Value)
   AlphaInterp = &RigVMFunction_AlphaInterp(InterpSpeedDecreasing=1, InterpSpeedIncreasing=2, Scale=1, bClampResult=False, bMapRange=true, Value=@Cast.Result)
   &RigUnit_ModifyTransforms(
    ItemToModify=((Item=(Type=Bone,Name="dfrm_back_strap_outer_mstr_l"),Transform=(Rotation=(X=0,Y=0,Z=0,W=1),Translation=(X=0,Y=0,Z=0),Scale3D=(X=1,Y=1,Z=1))))
    ItemToModify.0.Transform.Rotation=&RigVMFunction_MathQuaternionFromEuler(Euler=(X=-20,Y=-20,Z=-30)).Result
    ExecutePin=@SequenceAggregate_4.D
    Weight=@AlphaInterp.Result
   )
   &RigUnit_ModifyTransforms(
    ItemToModify=((Item=(Type=Bone,Name="dfrm_back_strap_inner_mstr_r"),Transform=(Rotation=(X=0,Y=0,Z=0,W=1),Translation=(X=0,Y=0,Z=0),Scale3D=(X=1,Y=1,Z=1))))
    ItemToModify.0.Transform.Rotation=&RigVMFunction_MathQuaternionFromEuler(Euler=(X=-10,Y=-10,Z=-15)).Result
    ExecutePin=@SequenceAggregate_4.B
    Weight=@AlphaInterp.Result
   )
   &RigUnit_ModifyTransforms(
    ItemToModify=((Item=(Type=Bone,Name="dfrm_back_strap_outer_mstr_r"),Transform=(Rotation=(X=0,Y=0,Z=0,W=1),Translation=(X=0,Y=0,Z=0),Scale3D=(X=1,Y=1,Z=1))))
    ItemToModify.0.Transform.Rotation=&RigVMFunction_MathQuaternionFromEuler(Euler=(X=-5,Y=-5,Z=-7.5)).Result
    ExecutePin=@SequenceAggregate_4.A
    Weight=@AlphaInterp.Result
   )
   &RigUnit_ModifyTransforms(
    ItemToModify=((Item=(Type=Bone,Name="dfrm_back_strap_inner_mstr_l"),Transform=(Rotation=(X=0,Y=0,Z=0,W=1),Translation=(X=0,Y=0,Z=0),Scale3D=(X=1,Y=1,Z=1))))
    ItemToModify.0.Transform.Rotation=&RigVMFunction_MathQuaternionFromEuler(Euler=(X=-15,Y=-15,Z=-22)).Result
    Weight=@AlphaInterp.Result
    ExecutePin=@SequenceAggregate_4.C
   )
   SequenceAggregate_6 = RigVMAggregateNode(ExecuteContext=@SequenceAggregate_1.D)
   &RigUnit_ModifyTransforms(
    ItemToModify=((Item=(Type=Bone,Name="dfrm_front_strap_outer_mstr_l"),Transform=(Rotation=(X=0,Y=0,Z=0,W=1),Translation=(X=-3,Y=0,Z=-1),Scale3D=(X=1,Y=1,Z=1))))
    ItemToModify.0.Transform.Rotation=&RigVMFunction_MathQuaternionFromEuler(Euler=(X=5,Y=5,Z=-25)).Result
    ExecutePin=@SequenceAggregate_6.A
    Weight=RigVMVariableNode() {variable=forward_interp_alpha}.Value
   )
   > &RigUnit_ModifyTransforms(
    ItemToModify=((Item=(Type=Bone,Name="dfrm_front_strap_outer_mstr_r"),Transform=(Rotation=(X=0,Y=0,Z=0,W=1),Translation=(X=-3,Y=0,Z=1),Scale3D=(X=1,Y=1,Z=1))))
    ItemToModify.0.Transform.Rotation=&RigVMFunction_MathQuaternionFromEuler(Euler=(X=-5,Y=-5,Z=-25)).Result
    Weight=RigVMVariableNode() {variable=forward_interp_alpha}.Value
   )
   n24 = &RigVMFunction_MathQuaternionFromEuler(Euler=(X=0,Y=0,Z=-75))
   > &RigUnit_ModifyTransforms(
    ItemToModify=((Item=(Type=Bone,Name="dfrm_front_strap_inner_mstr_l"),Transform=(Rotation=(X=0,Y=0,Z=0,W=1),Translation=(X=0,Y=0,Z=0),Scale3D=(X=1,Y=1,Z=1))))
    ItemToModify.0.Transform.Rotation=@n24.Result
    Weight=RigVMVariableNode() {variable=forward_interp_alpha}.Value
   )
   > &RigUnit_ModifyTransforms(
    ItemToModify=((Item=(Type=Bone,Name="dfrm_front_strap_inner_mstr_r"),Transform=(Rotation=(X=0,Y=0,Z=0,W=1),Translation=(X=0,Y=0,Z=0),Scale3D=(X=1,Y=1,Z=1))))
    Weight=RigVMVariableNode() {variable=forward_interp_alpha}.Value
    ItemToModify.0.Transform.Rotation=@n24.Result
   )
   &RigUnit_ModifyTransforms(
    ItemToModify=((Item=(Type=Bone,Name="dfrm_front_strap_outer_mstr_l"),Transform=(Rotation=(X=0,Y=0,Z=0,W=1),Translation=(X=-1,Y=0,Z=-1),Scale3D=(X=1,Y=1,Z=1))))
    ItemToModify.0.Transform.Rotation=&RigVMFunction_MathQuaternionFromEuler(Euler=(X=0,Y=0,Z=-10)).Result
    ExecutePin=@SequenceAggregate_6.C
    Weight=RigVMVariableNode() {variable=backward_interp_alpha}.Value
   )
   > &RigUnit_ModifyTransforms(
    ItemToModify=((Item=(Type=Bone,Name="dfrm_front_strap_outer_mstr_r"),Transform=(Rotation=(X=0,Y=0,Z=0,W=1),Translation=(X=-1,Y=0,Z=1),Scale3D=(X=1,Y=1,Z=1))))
    ItemToModify.0.Transform.Rotation=&RigVMFunction_MathQuaternionFromEuler(Euler=(X=0,Y=0,Z=-10)).Result
    Weight=RigVMVariableNode() {variable=backward_interp_alpha}.Value
   )
   &RigUnit_ModifyTransforms(
    ItemToModify=((Item=(Type=Bone,Name="dfrm_front_strap_outer_mstr_l"),Transform=(Rotation=(X=0,Y=0,Z=0,W=1),Translation=(X=-2.5,Y=0,Z=-1),Scale3D=(X=1,Y=1,Z=1))))
    ItemToModify.0.Transform.Rotation=&RigVMFunction_MathQuaternionFromEuler(Euler=(X=10,Y=10,Z=-15)).Result
    Weight=RigVMVariableNode() {variable=right_interp_alpha}.Value
    ExecutePin=@SequenceAggregate_6.E
   )
   > &RigUnit_ModifyTransforms(
    ItemToModify=((Item=(Type=Bone,Name="dfrm_front_strap_inner_mstr_l"),Transform=(Rotation=(X=0,Y=0,Z=0,W=1),Translation=(X=-6,Y=0,Z=2),Scale3D=(X=1,Y=1,Z=1))))
    ItemToModify.0.Transform.Rotation=&RigVMFunction_MathQuaternionFromEuler(Euler=(X=-10,Y=10,Z=-20)).Result
    Weight=RigVMVariableNode() {variable=right_interp_alpha}.Value
   )
   > &RigUnit_ModifyTransforms(
    ItemToModify=((Item=(Type=Bone,Name="dfrm_front_strap_inner_mstr_r"),Transform=(Rotation=(X=0,Y=0,Z=0,W=1),Translation=(X=-3.5,Y=0,Z=0.75),Scale3D=(X=1,Y=1,Z=1))))
    ItemToModify.0.Transform.Rotation=&RigVMFunction_MathQuaternionFromEuler(Euler=(X=-5,Y=5,Z=-10)).Result
    Weight=RigVMVariableNode() {variable=right_interp_alpha}.Value
   )
   &RigUnit_ModifyTransforms(
    ItemToModify=((Item=(Type=Bone,Name="dfrm_front_strap_outer_mstr_r"),Transform=(Rotation=(X=0,Y=0,Z=0,W=1),Translation=(X=-2.5,Y=0,Z=2),Scale3D=(X=1,Y=1,Z=1))))
    ItemToModify.0.Transform.Rotation=&RigVMFunction_MathQuaternionFromEuler(Euler=(X=-10,Y=-10,Z=-25)).Result
    Weight=RigVMVariableNode() {variable=left_interp_alpha}.Value
    ExecutePin=@SequenceAggregate_6.F
   )
   > &RigUnit_ModifyTransforms(
    ItemToModify=((Item=(Type=Bone,Name="dfrm_front_strap_inner_mstr_r"),Transform=(Rotation=(X=0,Y=0,Z=0,W=1),Translation=(X=-6,Y=0,Z=2),Scale3D=(X=1,Y=1,Z=1))))
    ItemToModify.0.Transform.Rotation=&RigVMFunction_MathQuaternionFromEuler(Euler=(X=10,Y=-10,Z=-20)).Result
    Weight=RigVMVariableNode() {variable=left_interp_alpha}.Value
   )
   > &RigUnit_ModifyTransforms(
    ItemToModify=((Item=(Type=Bone,Name="dfrm_front_strap_inner_mstr_l"),Transform=(Rotation=(X=0,Y=0,Z=0,W=1),Translation=(X=-3.5,Y=0,Z=0.75),Scale3D=(X=1,Y=1,Z=1))))
    ItemToModify.0.Transform.Rotation=&RigVMFunction_MathQuaternionFromEuler(Euler=(X=5,Y=-5,Z=-10)).Result
    Weight=RigVMVariableNode() {variable=left_interp_alpha}.Value
   )
   &RigUnit_ModifyTransforms(
    ItemToModify=((Item=(Type=Bone,Name="dfrm_front_strap_inner_mstr_l"),Transform=(Rotation=(X=0,Y=0,Z=0,W=1),Translation=(X=0,Y=0,Z=0),Scale3D=(X=1,Y=1,Z=1))))
    ItemToModify.0.Transform.Rotation=&RigVMFunction_MathQuaternionFromEuler(Euler=(X=0,Y=0,Z=-30)).Result
    Weight=RigVMVariableNode() {variable=left_right_interp_alpha}.Value
    ExecutePin=@SequenceAggregate_6.B
   )
   > &RigUnit_ModifyTransforms(
    ItemToModify=((Item=(Type=Bone,Name="dfrm_front_strap_inner_mstr_r"),Transform=(Rotation=(X=0,Y=0,Z=0,W=1),Translation=(X=0,Y=0,Z=0),Scale3D=(X=1,Y=1,Z=1))))
    Weight=RigVMVariableNode() {variable=left_right_interp_alpha}.Value
   )
   &RigUnit_ModifyTransforms(
    ItemToModify=((Item=(Type=Bone,Name="dfrm_front_strap_outer_mstr_l"),Transform=(Rotation=(X=0,Y=0,Z=0,W=1),Translation=(X=-3,Y=-0.75,Z=-0.5),Scale3D=(X=1,Y=1,Z=1))))
    ItemToModify.0.Transform.Rotation=&RigVMFunction_MathQuaternionFromEuler(Euler=(X=0,Y=25,Z=-75)).Result
    Weight=RigVMVariableNode() {variable=spr_thigh_l_up_alpha}.Value
    ExecutePin=@SequenceAggregate_1.B
   )
   > &RigUnit_ModifyTransforms(
    ItemToModify=((Item=(Type=Bone,Name="dfrm_front_strap_outer_mstr_r"),Transform=(Rotation=(X=0,Y=0,Z=0,W=1),Translation=(X=-3,Y=-0.75,Z=0.5),Scale3D=(X=1,Y=1,Z=1))))
    ItemToModify.0.Transform.Rotation=&RigVMFunction_MathQuaternionFromEuler(Euler=(X=0,Y=-25,Z=-75)).Result
    Weight=RigVMVariableNode() {variable=spr_thigh_r_up_alpha}.Value
   )
   > &RigUnit_ParentConstraint(
    AdvancedSettings=(RotationOrderForFilter=XZY)
    Child=(Type=Bone,Name="dfrm_front_strap_inner_mstr_l")
    Parents=((Item=(Type=Bone,Name="dfrm_front_strap_outer_mstr_l"),Weight=0.5),(Item=(Type=Bone,Name="dfrm_front_strap_outer_mstr_r"),Weight=0.25),(Item=(Type=Bone,Name="pelvis"),Weight=0.25))
    Weight=1
   )
   > &RigUnit_ParentConstraint(
    AdvancedSettings=(RotationOrderForFilter=XZY)
    Child=(Type=Bone,Name="dfrm_front_strap_inner_mstr_r")
    Parents=((Item=(Type=Bone,Name="dfrm_front_strap_outer_mstr_l"),Weight=0.25),(Item=(Type=Bone,Name="dfrm_front_strap_outer_mstr_r"),Weight=0.5),(Item=(Type=Bone,Name="pelvis"),Weight=0.25))
    Weight=1
   )
   SequenceAggregate_2 = RigVMAggregateNode(ExecuteContext=@SequenceAggregate_1_1.C)
   Branch_RigVMUnitNode = &RigVMFunction_ControlFlowBranch(ExecuteContext=@SequenceAggregate_2.A, Condition=RigVMVariableNode() {variable=is_helmet_active}.Value)
   RigVMVariableNode(Value=(X=0,Y=0,Z=-10), ExecuteContext=@Branch_RigVMUnitNode.False) {variable=collar_rotate_value}
   > RigVMVariableNode(Value=(X=0,Y=0,Z=0)) {variable=collar_back_translate_value}
   RigVMVariableNode(Value=(X=0,Y=0,Z=-30), ExecuteContext=@Branch_RigVMUnitNode.True) {variable=collar_rotate_value}
   > RigVMVariableNode(Value=(X=-1.5,Y=0,Z=0)) {variable=collar_back_translate_value}
   &RigUnit_ModifyTransforms(
    ItemToModify=((Item=(Type=Bone,Name="dfrm_collar_mstr"),Transform=(Rotation=(X=0,Y=0,Z=0,W=1),Translation=(X=0,Y=0,Z=0),Scale3D=(X=1,Y=1,Z=1))))
    ItemToModify.0.Transform.Rotation=&RigVMFunction_MathQuaternionFromEuler(Euler=RigVMVariableNode() {variable=collar_rotate_value}.Value).Result
    ItemToModify.0.Transform.Translation=RigVMVariableNode() {variable=collar_back_translate_value}.Value
    Weight=RigVMVariableNode() {variable=spr_head_alpha}.Value
    ExecutePin=@SequenceAggregate_2.B
   )
  ]}
]
hierarchy [
 root {loc=<0 0 0>, rot=<0 0 0 1>}
 pelvis {parent=root, loc=<0 -0.7071 93.1191>, rot=<0.0127 0.707 -0.0127 -0.707>}
 spine_01 {parent=pelvis, loc=<6.7494 -0.6388 0>, rot=<0 0 0.034 -0.9994>}
 spine_02 {parent=spine_01, loc=<7.2518 -0.035 0>, rot=<0 0 0.0238 -0.9997>}
 spine_03 {parent=spine_02, loc=<6.0104 0 0>, rot=<0 0 0.0149 -0.9999>}
 spine_04 {parent=spine_03, loc=<8.1024 0 0>, rot=<0 0 -0.0132 -0.9999>}
 spine_05 {parent=spine_04, loc=<10.0146 0 0>, rot=<0 0 -0.0015 -1>}
 neck_01 {parent=spine_05, loc=<11.3138 0.1529 0>, rot=<0 0 -0.2003 -0.9797>}
 neck_02 {parent=neck_01, loc=<4.5627 0 0>, rot=<0 0 0.0504 -0.9987>}
 head {parent=neck_02, loc=<5.3914 0 0>, rot=<0 0 0.115 -0.9934>}
 attach_cape {parent=spine_05, loc=<9.017 -11.714 0>, rot=<0.0616 -0.7044 -0.0616 -0.7044>}
 clavicle_l {parent=spine_05, loc=<4.3575 -0.1466 -2.3223>, rot=<0.0465 -0.7289 -0.009 -0.6829>}
 upperarm_l {parent=clavicle_l, loc=<15.5111 0 0>, rot=<0.0226 -0.3234 -0.0088 -0.9459>}
 lowerarm_l {parent=upperarm_l, loc=<26.6514 0 0>, rot=<0.0391 -0.0099 -0.2453 -0.9686>}
 hand_l {parent=lowerarm_l, loc=<22.3491 0 0>, rot=<-0.7081 -0.0507 0.0975 -0.6975>}
 thumb_01_l {parent=hand_l, loc=<1.3118 -0.3191 -2.459>, rot=<0.6546 -0.4042 0.0664 -0.6354>}
 thumb_02_l {parent=thumb_01_l, loc=<5.9336 -0.0448 0.0401>, rot=<-0.0037 0.0213 0.298 -0.9543>}
 thumb_03_l {parent=thumb_02_l, loc=<2.2223 0.3682 -0.0907>, rot=<0.0118 -0.013 0.0066 -0.9998>}
 index_metacarpal_l {parent=hand_l, loc=<1.6535 0.0183 -2.96>, rot=<0.0158 -0.0124 -0.0173 -0.9996>}
 index_01_l {parent=index_metacarpal_l, loc=<8.3008 0.0357 -0.0524>, rot=<-0.0044 0.0485 0.0924 -0.9945>}
 index_02_l {parent=index_01_l, loc=<3.3587 -0.3004 0.0452>, rot=<-0.002 0.0027 0.3225 -0.9466>}
 index_03_l {parent=index_02_l, loc=<2.5578 -0.123 -0.0252>, rot=<-0.002 0.0067 0.2558 -0.9667>}
 middle_metacarpal_l {parent=hand_l, loc=<2.1856 -0.1725 -0.5984>, rot=<0.006 0.0168 -0.0329 -0.9993>}
 middle_01_l {parent=middle_metacarpal_l, loc=<7.3275 -0.2447 0.0129>, rot=<-0.0208 0.0042 0.0858 -0.9961>}
 middle_02_l {parent=middle_01_l, loc=<4.0123 0.1401 0.1175>, rot=<-0.0218 0.0258 0.3512 -0.9357>}
 middle_03_l {parent=middle_02_l, loc=<2.6077 0.0534 0.1373>, rot=<0.0082 0.0188 0.1708 -0.9851>}
 ring_metacarpal_l {parent=hand_l, loc=<2.2074 -0.3625 1.3638>, rot=<-0.0284 0.0722 -0.0458 -0.9959>}
 ring_01_l {parent=ring_metacarpal_l, loc=<7.0295 -0.6197 -0.1446>, rot=<-0.0489 -0.0252 0.1892 -0.9804>}
 ring_02_l {parent=ring_01_l, loc=<3.6174 0.4647 -0.1027>, rot=<0 0 0.2806 -0.9598>}
 ring_03_l {parent=ring_02_l, loc=<2.4152 0.114 0.0199>, rot=<0 0 0.1189 -0.9929>}
 pinky_metacarpal_l {parent=hand_l, loc=<1.7616 -0.9753 3.3814>, rot=<-0.0397 0.0814 -0.069 -0.9935>}
 pinky_01_l {parent=pinky_metacarpal_l, loc=<7.1709 -0.4831 -0.1094>, rot=<-0.0654 -0.0035 0.2242 -0.9723>}
 pinky_02_l {parent=pinky_01_l, loc=<2.8274 0.0893 -0.1043>, rot=<0.0037 0.0011 0.2854 -0.9584>}
 pinky_03_l {parent=pinky_02_l, loc=<2.2046 0.0307 -0.0078>, rot=<-0.0002 0.0098 0.1445 -0.9895>}
 deform_wrist_dn_base_l {parent=hand_l, loc=<0 -2.2342 0>, rot=<0 0 0 -1>}
 deform_wrist_dn_l {parent=deform_wrist_dn_base_l, loc=<0 0 0>, rot=<0 0 0 1>}
 deform_wrist_up_base_l {parent=hand_l, loc=<0 2.2682 0>, rot=<0 0 0 -1>}
 deform_wrist_up_l {parent=deform_wrist_up_base_l, loc=<0 0 0>, rot=<0 0 0 1>}
 lowerarm_twist_02_l {parent=lowerarm_l, loc=<9.6199 0 0>, rot=<0 0 0 -1>}
 lowerarm_twist_01_l {parent=lowerarm_l, loc=<17.8549 0 0>, rot=<0 0 0 -1>}
 deform_elbow_in_base_l {parent=lowerarm_l, loc=<0.2112 5.259 0.8274>, rot=<0 0 0 -1>}
 deform_elbow_in_l {parent=deform_elbow_in_base_l, loc=<0 0 0>, rot=<0 0 0 1>}
 deform_elbow_out_base_l {parent=lowerarm_l, loc=<0 -4 0>, rot=<0 0 0 -1>}
 deform_elbow_out_l {parent=deform_elbow_out_base_l, loc=<0 0 0>, rot=<0 0 0 1>}
 upperarm_twist_01_l {parent=upperarm_l, loc=<0.0041 0 0>, rot=<0 0 0 -1>}
 deform_upperarm_bk_base_l {parent=upperarm_twist_01_l, loc=<1.0813 -7.2193 -1.9453>, rot=<0 0 0 -1>}
 deform_upperarm_bk_l {parent=deform_upperarm_bk_base_l, loc=<0 0 0>, rot=<0 0 0 1>}
 deform_upperarm_fr_base_l {parent=upperarm_twist_01_l, loc=<-0.3691 7.0506 -2.765>, rot=<0 0 0 -1>}
 deform_upperarm_fr_l {parent=deform_upperarm_fr_base_l, loc=<0 0 0>, rot=<0 0 0 1>}
 deform_upperarm_up_base_l {parent=upperarm_twist_01_l, loc=<-3.2216 -1.1391 3.3311>, rot=<0 0 0 -1>}
 deform_upperarm_up_l {parent=deform_upperarm_up_base_l, loc=<0 0 0>, rot=<0 0 0 1>}
 upperarm_twist_02_l {parent=upperarm_l, loc=<13.7941 0 0>, rot=<0 0 0 -1>}
 dfrm_clavicle_l {parent=clavicle_l, loc=<0 0 0>, rot=<0 0 0 1>}
 attach_backpack {parent=spine_05, loc=<-0.123 -13.473 0>, rot=<-0.0093 -0.707 0.0093 -0.707>}
 clavicle_r {parent=spine_05, loc=<4.3575 -0.1466 2.3223>, rot=<0.7289 0.0465 -0.6829 0.009>}
 upperarm_r {parent=clavicle_r, loc=<-15.5111 0 0>, rot=<0.0226 -0.3234 -0.0088 -0.9459>}
 lowerarm_r {parent=upperarm_r, loc=<-26.6514 0 0>, rot=<0.0391 -0.0099 -0.2453 -0.9686>}
 hand_r {parent=lowerarm_r, loc=<-22.3491 0 0>, rot=<-0.7081 -0.0507 0.0975 -0.6975>}
 thumb_01_r {parent=hand_r, loc=<-1.3118 0.3191 2.459>, rot=<0.6546 -0.4042 0.0664 -0.6354>}
 thumb_02_r {parent=thumb_01_r, loc=<-5.9336 0.0448 -0.0401>, rot=<-0.0037 0.0213 0.298 -0.9543>}
 thumb_03_r {parent=thumb_02_r, loc=<-2.2223 -0.3682 0.0907>, rot=<0.0118 -0.013 0.0066 -0.9998>}
 index_metacarpal_r {parent=hand_r, loc=<-1.6535 -0.0183 2.96>, rot=<0.0158 -0.0124 -0.0173 -0.9996>}
 index_01_r {parent=index_metacarpal_r, loc=<-8.3008 -0.0357 0.0524>, rot=<-0.0044 0.0485 0.0924 -0.9945>}
 index_02_r {parent=index_01_r, loc=<-3.3587 0.3004 -0.0452>, rot=<-0.002 0.0027 0.3225 -0.9466>}
 index_03_r {parent=index_02_r, loc=<-2.5578 0.123 0.0252>, rot=<-0.002 0.0067 0.2558 -0.9667>}
 middle_metacarpal_r {parent=hand_r, loc=<-2.1856 0.1725 0.5984>, rot=<0.006 0.0168 -0.0329 -0.9993>}
 middle_01_r {parent=middle_metacarpal_r, loc=<-7.3275 0.2447 -0.0129>, rot=<-0.0208 0.0042 0.0858 -0.9961>}
 middle_02_r {parent=middle_01_r, loc=<-4.0123 -0.1401 -0.1175>, rot=<-0.0218 0.0258 0.3512 -0.9357>}
 middle_03_r {parent=middle_02_r, loc=<-2.6077 -0.0534 -0.1373>, rot=<0.0082 0.0188 0.1708 -0.9851>}
 ring_metacarpal_r {parent=hand_r, loc=<-2.2074 0.3625 -1.3638>, rot=<-0.0284 0.0722 -0.0458 -0.9959>}
 ring_01_r {parent=ring_metacarpal_r, loc=<-7.0295 0.6197 0.1446>, rot=<-0.0489 -0.0252 0.1892 -0.9804>}
 ring_02_r {parent=ring_01_r, loc=<-3.6174 -0.4647 0.1027>, rot=<0 0 0.2806 -0.9598>}
 ring_03_r {parent=ring_02_r, loc=<-2.4152 -0.114 -0.0199>, rot=<0 0 0.1189 -0.9929>}
 pinky_metacarpal_r {parent=hand_r, loc=<-1.7616 0.9753 -3.3814>, rot=<-0.0397 0.0814 -0.069 -0.9935>}
 pinky_01_r {parent=pinky_metacarpal_r, loc=<-7.1709 0.4831 0.1094>, rot=<-0.0654 -0.0035 0.2242 -0.9723>}
 pinky_02_r {parent=pinky_01_r, loc=<-2.8274 -0.0893 0.1043>, rot=<0.0037 0.0011 0.2854 -0.9584>}
 pinky_03_r {parent=pinky_02_r, loc=<-2.2046 -0.0307 0.0078>, rot=<-0.0002 0.0098 0.1445 -0.9895>}
 deform_wrist_dn_base_r {parent=hand_r, loc=<-0.0004 2.2345 -0.0001>, rot=<0 0 0 -1>}
 deform_wrist_dn_r {parent=deform_wrist_dn_base_r, loc=<0 0 0>, rot=<0 0 0 1>}
 deform_wrist_up_base_r {parent=hand_r, loc=<0.0003 -2.2684 0.0001>, rot=<0 0 0 -1>}
 deform_wrist_up_r {parent=deform_wrist_up_base_r, loc=<0 0 0>, rot=<0 0 0 1>}
 lowerarm_twist_02_r {parent=lowerarm_r, loc=<-9.6199 0 0>, rot=<0 0 0 -1>}
 lowerarm_twist_01_r {parent=lowerarm_r, loc=<-17.8549 0 0>, rot=<0 0 0 -1>}
 deform_elbow_in_base_r {parent=lowerarm_r, loc=<-0.2111 -5.2591 -0.8275>, rot=<0 0 0 -1>}
 deform_elbow_in_r {parent=deform_elbow_in_base_r, loc=<-0.0001 0 0.0002>, rot=<0 0 0 1>}
 deform_elbow_out_base_r {parent=lowerarm_r, loc=<0 4 -0.0001>, rot=<0 0 0 -1>}
 deform_elbow_out_r {parent=deform_elbow_out_base_r, loc=<0 0 0>, rot=<0 0 0 1>}
 upperarm_twist_01_r {parent=upperarm_r, loc=<-0.0041 0 0>, rot=<0 0 0 -1>}
 deform_upperarm_bk_base_r {parent=upperarm_twist_01_r, loc=<-1.0814 7.2192 1.9455>, rot=<0 0 0 1>}
 deform_upperarm_bk_r {parent=deform_upperarm_bk_base_r, loc=<0 0 0>, rot=<0 0 0 1>}
 deform_upperarm_fr_base_r {parent=upperarm_twist_01_r, loc=<0.3689 -7.0506 2.7652>, rot=<0 0 0 1>}
 deform_upperarm_fr_r {parent=deform_upperarm_fr_base_r, loc=<0 0 0>, rot=<0 0 0 1>}
 deform_upperarm_up_base_r {parent=upperarm_twist_01_r, loc=<3.2214 1.1391 -3.3308>, rot=<0 0 0 1>}
 deform_upperarm_up_r {parent=deform_upperarm_up_base_r, loc=<0 0 0>, rot=<0 0 0 -1>}
 upperarm_twist_02_r {parent=upperarm_r, loc=<-13.7941 0 0>, rot=<0 0 0 -1>}
 dfrm_clavicle_r {parent=clavicle_r, loc=<0 0 0>, rot=<0 0 0 1>}
 dfrm_collar_mstr {parent=spine_05, loc=<11.3138 0.1529 0>, rot=<0 0 -0.2003 -0.9797>}
 dfrm_collar_back {parent=dfrm_collar_mstr, loc=<6.26 -9.1537 0>, rot=<0.8567 -0.5155 0.0186 0.0006>}
 dfrm_collar_back_l {parent=dfrm_collar_mstr, loc=<6.0963 -7.6958 -6.378>, rot=<0.8201 -0.519 -0.1406 -0.1956>}
 dfrm_collar_back_r {parent=dfrm_collar_mstr, loc=<6.0965 -7.6958 6.378>, rot=<0.8201 -0.519 0.1406 0.1956>}
 deform_pec_aim_base_l {parent=spine_04, loc=<7.8732 12.6388 -2.4123>, rot=<0.0102 -0.4602 0.3979 -0.7936>}
 deform_pec_aim_l {parent=deform_pec_aim_base_l, loc=<0 0 0>, rot=<0 0 0 1>}
 deform_pec_base_l {parent=deform_pec_aim_l, loc=<10.766 0 0>, rot=<0 0 0 1>}
 deform_pec_l {parent=deform_pec_base_l, loc=<0.055 0 0>, rot=<0 0 0 1>}
 deform_pec_aim_base_r {parent=spine_04, loc=<7.8727 12.6388 2.4123>, rot=<-0.4602 -0.0102 0.7936 0.3979>}
 deform_pec_aim_r {parent=deform_pec_aim_base_r, loc=<0 0 0>, rot=<0 0 0 1>}
 deform_pec_base_r {parent=deform_pec_aim_r, loc=<-10.7662 -0.0006 -0.0007>, rot=<0 0 0 1>}
 deform_pec_r {parent=deform_pec_base_r, loc=<0 0 0>, rot=<0 0 0 1>}
 thigh_l {parent=pelvis, loc=<-2.5693 -0.0109 -7.7501>, rot=<0.0012 0.0572 0.0155 -0.9982>}
 calf_l {parent=thigh_l, loc=<-39.0484 0 0>, rot=<0.0006 -0.0161 -0.0391 -0.9991>}
 foot_l {parent=calf_l, loc=<-42.561 0.1077 0.4292>, rot=<0.0418 -0.0455 0.0343 -0.9975>}
 ball_l {parent=foot_l, loc=<-6.1599 13.8264 0.6806>, rot=<0.0001 -0.0001 -0.7063 -0.7079>}
 calf_twist_02_l {parent=calf_l, loc=<-18.8064 -0.0967 0.0397>, rot=<0.0237 0.0095 -0.0035 -0.9997>}
 calf_twist_01_l {parent=calf_l, loc=<-35.813 -0.0967 0.0397>, rot=<0.0237 0.0095 -0.0035 -0.9997>}
 deform_knee_in_base_l {parent=calf_l, loc=<1.298 -5.2957 -0.0496>, rot=<0 0 0 -1>}
 deform_knee_in_l {parent=deform_knee_in_base_l, loc=<0 0 0>, rot=<0 0 0 1>}
 deform_knee_out_base_l {parent=calf_l, loc=<0 4 0>, rot=<0 0 0 -1>}
 deform_knee_out_l {parent=deform_knee_out_base_l, loc=<0 0 0>, rot=<0 0 0 1>}
 thigh_twist_01_l {parent=thigh_l, loc=<-21.6303 0 0>, rot=<0.0013 0 0 -1>}
 deform_groin_base_l {parent=thigh_l, loc=<-2.1908 8.1142 -1.7442>, rot=<0 0 0 -1>}
 deform_groin_l {parent=deform_groin_base_l, loc=<0 0 0>, rot=<0 0 0 1>}
 thigh_r {parent=pelvis, loc=<-2.5693 -0.0109 7.7501>, rot=<0.0572 -0.0012 0.9982 0.0155>}
 calf_r {parent=thigh_r, loc=<39.0484 0 0>, rot=<0.0006 -0.0161 -0.0391 -0.9991>}
 foot_r {parent=calf_r, loc=<42.561 -0.1077 -0.4292>, rot=<0.0418 -0.0455 0.0343 -0.9975>}
 ball_r {parent=foot_r, loc=<6.1599 -13.8264 -0.6806>, rot=<0.0001 -0.0001 -0.7063 -0.7079>}
 calf_twist_02_r {parent=calf_r, loc=<18.8064 0.0967 -0.0397>, rot=<0.0237 0.0095 -0.0035 -0.9997>}
 calf_twist_01_r {parent=calf_r, loc=<35.813 0.0967 -0.0397>, rot=<0.0237 0.0095 -0.0035 -0.9997>}
 deform_knee_in_base_r {parent=calf_r, loc=<-1.298 5.2957 0.0496>, rot=<0 0 0 -1>}
 deform_knee_in_r {parent=deform_knee_in_base_r, loc=<0 0 0>, rot=<0 0 0 1>}
 deform_knee_out_base_r {parent=calf_r, loc=<0 -4 0>, rot=<0 0 0 -1>}
 deform_knee_out_r {parent=deform_knee_out_base_r, loc=<0 0 0>, rot=<0 0 0 1>}
 thigh_twist_01_r {parent=thigh_r, loc=<21.6303 0 0>, rot=<0.0013 0 0 -1>}
 deform_groin_base_r {parent=thigh_r, loc=<2.1908 -8.1142 1.7442>, rot=<0 0 0 -1>}
 deform_groin_r {parent=deform_groin_base_r, loc=<0 0 0>, rot=<0 0 0 1>}
 deform_glute_base_l {parent=pelvis, loc=<-8.8064 -8.8222 -8.4287>, rot=<0.0275 0.0501 -0.4472 -0.8926>}
 deform_glute_l {parent=deform_glute_base_l, loc=<0 0 0>, rot=<0 0 0 1>}
 deform_glute_base_r {parent=pelvis, loc=<-8.8062 -8.8222 8.4287>, rot=<0.0501 -0.0275 0.8926 -0.4472>}
 deform_glute_r {parent=deform_glute_base_r, loc=<0 0 0>, rot=<0 0 0 1>}
 dfrm_front_strap_outer_mstr_l {parent=pelvis, loc=<2.6066 8.1826 -5.4999>, rot=<-0.0394 0.0614 0.9938 0.0834>}
 dfrm_front_strap_outer_01_l {parent=dfrm_front_strap_outer_mstr_l, loc=<6.6674 0 0>, rot=<0.0597 -0.0073 -0.0078 -0.9982>}
 dfrm_front_strap_outer_02_l {parent=dfrm_front_strap_outer_01_l, loc=<6.3712 0 0>, rot=<0 0 -0.023 -0.9997>}
 dfrm_front_strap_outer_03_l {parent=dfrm_front_strap_outer_02_l, loc=<6.3712 0 0>, rot=<0 0 -0.0202 -0.9998>}
 dfrm_front_strap_outer_04_l {parent=dfrm_front_strap_outer_03_l, loc=<6.3712 0 0>, rot=<0 0 -0.0215 -0.9998>}
 dfrm_front_strap_inner_mstr_l {parent=pelvis, loc=<2.4486 8.9948 -2.1501>, rot=<-0.0038 0.0315 0.9974 0.0647>}
 dfrm_front_strap_inner_01_l {parent=dfrm_front_strap_inner_mstr_l, loc=<7.4969 0 0>, rot=<0.0115 -0.0081 -0.0109 -0.9998>}
 dfrm_front_strap_inner_02_l {parent=dfrm_front_strap_inner_01_l, loc=<7.1339 0 0>, rot=<-0.0029 -0.0026 -0.0166 -0.9999>}
 dfrm_front_strap_inner_03_l {parent=dfrm_front_strap_inner_02_l, loc=<7.1339 0 0>, rot=<0 0 -0.0076 -1>}
 dfrm_front_strap_inner_04_l {parent=dfrm_front_strap_inner_03_l, loc=<7.1339 0 0>, rot=<0 0 -0.0008 -1>}
 dfrm_front_strap_outer_mstr_r {parent=pelvis, loc=<2.6067 8.1826 5.4999>, rot=<0.0394 -0.0614 0.9938 0.0834>}
 dfrm_front_strap_outer_01_r {parent=dfrm_front_strap_outer_mstr_r, loc=<6.6674 0 0>, rot=<-0.0597 0.0073 -0.0078 -0.9982>}
 dfrm_front_strap_outer_02_r {parent=dfrm_front_strap_outer_01_r, loc=<6.3712 0 0>, rot=<0 0 -0.023 -0.9997>}
 dfrm_front_strap_outer_03_r {parent=dfrm_front_strap_outer_02_r, loc=<6.3713 0 0>, rot=<0 0 -0.0202 -0.9998>}
 dfrm_front_strap_outer_04_r {parent=dfrm_front_strap_outer_03_r, loc=<6.3713 0 0>, rot=<0 0 -0.0215 -0.9998>}
 dfrm_front_strap_inner_mstr_r {parent=pelvis, loc=<2.4486 8.9948 2.1501>, rot=<0.0038 -0.0315 0.9974 0.0647>}
 dfrm_front_strap_inner_01_r {parent=dfrm_front_strap_inner_mstr_r, loc=<7.4968 0 0>, rot=<-0.0115 0.0081 -0.0109 -0.9998>}
 dfrm_front_strap_inner_02_r {parent=dfrm_front_strap_inner_01_r, loc=<7.1339 0 0>, rot=<0.0029 0.0026 -0.0166 -0.9999>}
 dfrm_front_strap_inner_03_r {parent=dfrm_front_strap_inner_02_r, loc=<7.134 0 0>, rot=<0 0 -0.0076 -1>}
 dfrm_front_strap_inner_04_r {parent=dfrm_front_strap_inner_03_r, loc=<7.1339 0 0>, rot=<0 0 -0.0008 -1>}
 dfrm_back_strap_outer_mstr_l {parent=pelvis, loc=<5.2921 -10.6769 -5.5591>, rot=<0.0942 -0.9944 -0.0139 -0.0467>}
 dfrm_back_strap_outer_01_l {parent=dfrm_back_strap_outer_mstr_l, loc=<5.6912 0 0>, rot=<0.0045 0.0032 -0.0368 -0.9993>}
 dfrm_back_strap_outer_02_l {parent=dfrm_back_strap_outer_01_l, loc=<8.8039 0 0>, rot=<0 0 -0.0589 -0.9983>}
 dfrm_back_strap_outer_03_l {parent=dfrm_back_strap_outer_02_l, loc=<8.8039 0 0>, rot=<0 0 -0.0275 -0.9996>}
 dfrm_back_strap_inner_mstr_l {parent=pelvis, loc=<5.3688 -10.8174 -1.889>, rot=<0.1079 -0.9941 0.0018 -0.0127>}
 dfrm_back_strap_inner_01_l {parent=dfrm_back_strap_inner_mstr_l, loc=<6.2576 0 0>, rot=<-0.0271 0.0021 -0.0506 -0.9983>}
 dfrm_back_strap_inner_02_l {parent=dfrm_back_strap_inner_01_l, loc=<10.0018 0 0>, rot=<0.0278 -0.0021 -0.0751 -0.9968>}
 dfrm_back_strap_inner_03_l {parent=dfrm_back_strap_inner_02_l, loc=<10.0018 0 0>, rot=<0 0 -0.0198 -0.9998>}
 dfrm_back_strap_inner_mstr_r {parent=pelvis, loc=<5.3688 -10.8173 1.889>, rot=<0.1079 -0.9941 -0.0018 0.0127>}
 dfrm_back_strap_inner_01_r {parent=dfrm_back_strap_inner_mstr_r, loc=<6.2576 0 0>, rot=<0.0271 -0.0021 -0.0506 -0.9983>}
 dfrm_back_strap_inner_02_r {parent=dfrm_back_strap_inner_01_r, loc=<10.0019 0 0>, rot=<-0.0278 0.0021 -0.0751 -0.9968>}
 dfrm_back_strap_inner_03_r {parent=dfrm_back_strap_inner_02_r, loc=<10.0018 0 0>, rot=<0 0 -0.0198 -0.9998>}
 dfrm_back_strap_outer_mstr_r {parent=pelvis, loc=<5.2921 -10.6769 5.5591>, rot=<0.0942 -0.9944 0.0139 0.0467>}
 dfrm_back_strap_outer_01_r {parent=dfrm_back_strap_outer_mstr_r, loc=<5.6911 0 0>, rot=<-0.0045 -0.0032 -0.0368 -0.9993>}
 dfrm_back_strap_outer_02_r {parent=dfrm_back_strap_outer_01_r, loc=<8.8039 0 0>, rot=<0 0 -0.0589 -0.9983>}
 dfrm_back_strap_outer_03_r {parent=dfrm_back_strap_outer_02_r, loc=<8.8039 0 0>, rot=<0 0 -0.0275 -0.9996>}
]
variables [
 backward_interp_alpha {type=double, default=0}
 backwards_alpha {type=double, default=0}
 collar_back_translate_value {type=FVector, default=(X=0,Y=0,Z=0)}
 collar_rotate_value {type=FVector, default=(X=0,Y=0,Z=0)}
 forward_alpha {type=double, default=0}
 forward_backwards_movement_alpha {type=double, default=0}
 forward_interp_alpha {type=double, default=0}
 head_tilt_aim_alpha {type=double, default=0}
 is_crouching {type=bool, default=False}
 is_helmet_active {type=bool, default=False}
 is_jumping {type=bool, default=False}
 left_alpha {type=double, default=0}
 left_interp_alpha {type=double, default=0}
 left_right_interp_alpha {type=double, default=0}
 left_right_movement_alpha {type=double, default=0}
 right_alpha {type=double, default=0}
 right_interp_alpha {type=double, default=0}
 spr_clavicle_r_twist {type=double, default=0}
 spr_head_alpha {type=float, default=1}
 spr_head_tilt_alpha {type=double, default=0}
 spr_thigh_l_up_alpha {type=double, default=0}
 spr_thigh_r_up_alpha {type=double, default=0}
 spr_upperarm_l_aim_pistol_alpha {type=double, default=0}
 spr_upperarm_l_aim_rifle_alpha {type=double, default=0}
]
