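# Physics asset for the CoyoteTrailDark medium male player body, defining dangling front and back strap chains on both sides.
# Each inner and outer strap is a chain of box bodies, four links in front and three in back, rooted at a small master capsule.
# Swing and twist-limited constraints join the links, with back straps bending widest at the root.
# Pelvis and thigh capsules supply body volumes, with no constraints attached.
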PhysicsAsset /BRCosmetics/Characters/Player/Male/Medium/Bodies/M_MED_CoyoteTrailDark/Meshes/M_MED_CoyoteTrailDark_Physics
bodies {
 dfrm_back_strap_inner_01_l [{shapeType=Box, extentX=10.1128, extentY=1.2, extentZ=6.5106, shapeName=dfrm_back_strap_inner_01_l_box, center=<5.3056 0 0>}]
 dfrm_back_strap_inner_01_r [{shapeType=Box, extentX=10.1128, extentY=1.2, extentZ=6.5106, shapeName=dfrm_back_strap_inner_01_l_box, center=<5.3056 0 0>}]
 dfrm_back_strap_inner_02_l [{shapeType=Box, extentX=10.1128, extentY=1.2, extentZ=6.5106, shapeName=dfrm_back_strap_inner_02_l_box, center=<5.3056 0 0>}]
 dfrm_back_strap_inner_02_r [{shapeType=Box, extentX=10.1128, extentY=1.2, extentZ=6.5106, shapeName=dfrm_back_strap_inner_02_l_box, center=<5.3056 0 0>}]
 dfrm_back_strap_inner_03_l [{shapeType=Box, extentX=10.1128, extentY=1.2, extentZ=6.5106, shapeName=dfrm_back_strap_inner_03_l_box, center=<5.3056 0 0>}]
 dfrm_back_strap_inner_03_r [{shapeType=Box, extentX=10.1128, extentY=1.2, extentZ=6.5106, shapeName=dfrm_back_strap_inner_03_l_box, center=<5.3056 0 0>}]
 dfrm_back_strap_inner_mstr_l [{shapeType=Capsule, length=0, radius=0.5, shapeName=dfrm_back_strap_inner_mstr_l_capsule}]
 dfrm_back_strap_inner_mstr_r [{shapeType=Capsule, length=0, radius=0.5, shapeName=dfrm_back_strap_inner_mstr_l_capsule, rotation=<rot 0 0 -180>}]
 dfrm_back_strap_outer_01_l [{shapeType=Box, extentX=10.1128, extentY=1.2, extentZ=6.5106, shapeName=dfrm_back_strap_outer_01_l_box, center=<5.3056 0 0>}]
 dfrm_back_strap_outer_01_r [{shapeType=Box, extentX=10.1128, extentY=1.2, extentZ=6.5106, shapeName=dfrm_back_strap_outer_01_l_box, center=<5.3056 0 0>}]
 dfrm_back_strap_outer_02_l [{shapeType=Box, extentX=10.1128, extentY=1.2, extentZ=6.5106, shapeName=dfrm_back_strap_outer_02_l_box, center=<5.3056 0 0>}]
 dfrm_back_strap_outer_02_r [{shapeType=Box, extentX=10.1128, extentY=1.2, extentZ=6.5106, shapeName=dfrm_back_strap_outer_02_l_box, center=<5.3056 0 0>}]
 dfrm_back_strap_outer_03_l [{shapeType=Box, extentX=10.1128, extentY=1.2, extentZ=6.5106, shapeName=dfrm_back_strap_outer_03_l_box, center=<5.3056 0 0>}]
 dfrm_back_strap_outer_03_r [{shapeType=Box, extentX=10.1128, extentY=1.2, extentZ=6.5106, shapeName=dfrm_back_strap_outer_03_l_box, center=<5.3056 0 0>}]
 dfrm_back_strap_outer_mstr_l [{shapeType=Capsule, length=0, radius=0.5, shapeName=dfrm_back_strap_outer_mstr_l_capsule}]
 dfrm_back_strap_outer_mstr_r [{shapeType=Capsule, length=0, radius=0.5, shapeName=dfrm_back_strap_outer_mstr_l_capsule, rotation=<rot 0 0 -180>}]
 dfrm_front_strap_inner_01_l [{shapeType=Box, extentX=7.8263, extentY=1.3, extentZ=6.1368, shapeName=dfrm_front_strap_inner_01_l_box, center=<3.7166 0 0>}]
 dfrm_front_strap_inner_01_r [{shapeType=Box, extentX=7.8263, extentY=1.3, extentZ=6.1368, shapeName=dfrm_front_strap_inner_01_l_box, center=<3.7166 0 0>}]
 dfrm_front_strap_inner_02_l [{shapeType=Box, extentX=7.8263, extentY=1.3, extentZ=6.1368, shapeName=dfrm_front_strap_inner_02_l_box, center=<3.7166 0 0>}]
 dfrm_front_strap_inner_02_r [{shapeType=Box, extentX=7.8263, extentY=1.3, extentZ=6.1368, shapeName=dfrm_front_strap_inner_02_l_box, center=<3.7166 0 0>}]
 dfrm_front_strap_inner_03_l [{shapeType=Box, extentX=7.8263, extentY=1.3, extentZ=6.1368, shapeName=dfrm_front_strap_inner_03_l_box, center=<3.7166 0 0>}]
 dfrm_front_strap_inner_03_r [{shapeType=Box, extentX=7.8263, extentY=1.3, extentZ=6.1368, shapeName=dfrm_front_strap_inner_03_l_box, center=<3.7166 0 0>}]
 dfrm_front_strap_inner_04_l [{shapeType=Box, extentX=7.8263, extentY=1.3, extentZ=6.1368, shapeName=dfrm_front_strap_inner_04_l_box, center=<3.7166 0 0>}]
 dfrm_front_strap_inner_04_r [{shapeType=Box, extentX=7.8263, extentY=1.3, extentZ=6.1368, shapeName=dfrm_front_strap_inner_04_l_box, center=<3.7166 0 0>}]
 dfrm_front_strap_inner_mstr_l [{shapeType=Capsule, length=0, radius=0.5, shapeName=dfrm_front_strap_inner_mstr_l_capsule}]
 dfrm_front_strap_inner_mstr_r [{shapeType=Capsule, length=0, radius=0.5, shapeName=dfrm_front_strap_inner_mstr_r_capsule}]
 dfrm_front_strap_outer_01_l [{shapeType=Box, extentX=6.6191, extentY=1.3, extentZ=5.0747, shapeName=dfrm_front_strap_outer_01_l_box, center=<3.1742 0 0>}]
 dfrm_front_strap_outer_01_r [{shapeType=Box, extentX=6.6191, extentY=1.3, extentZ=5.0747, shapeName=dfrm_front_strap_outer_01_l_box, center=<3.1742 0 0>}]
 dfrm_front_strap_outer_02_l [{shapeType=Box, extentX=6.6191, extentY=1.3, extentZ=5.0747, shapeName=dfrm_front_strap_outer_02_l_box, center=<3.1742 0 0>}]
 dfrm_front_strap_outer_02_r [{shapeType=Box, extentX=6.6191, extentY=1.3, extentZ=5.0747, shapeName=dfrm_front_strap_outer_02_l_box, center=<3.1742 0 0>}]
 dfrm_front_strap_outer_03_l [{shapeType=Box, extentX=6.6191, extentY=1.3, extentZ=5.0747, shapeName=dfrm_front_strap_outer_03_l_box, center=<3.1742 0 0>}]
 dfrm_front_strap_outer_03_r [{shapeType=Box, extentX=6.6191, extentY=1.3, extentZ=5.0747, shapeName=dfrm_front_strap_outer_03_l_box, center=<3.1742 0 0>}]
 dfrm_front_strap_outer_04_l [{shapeType=Box, extentX=6.6191, extentY=1.3, extentZ=5.0747, shapeName=dfrm_front_strap_outer_04_l_box, center=<3.1742 0 0>}]
 dfrm_front_strap_outer_04_r [{shapeType=Box, extentX=6.6191, extentY=1.3, extentZ=5.0747, shapeName=dfrm_front_strap_outer_04_l_box, center=<3.1742 0 0>}]
 dfrm_front_strap_outer_mstr_l [{shapeType=Capsule, length=0, radius=0.5, shapeName=dfrm_front_strap_outer_mstr_l_capsule}]
 dfrm_front_strap_outer_mstr_r [{shapeType=Capsule, length=0, radius=0.5, shapeName=dfrm_front_strap_outer_mstr_r_capsule}]
 pelvis [
  {shapeType=Capsule, length=7.8045, radius=7.9636, shapeName=pelvis_capsule, center=<-4.1412 -3.2054 -3.4851>, rotation=<rot 0 -0.0006 -5>}
  {shapeType=Capsule, length=7.8045, radius=7.9636, shapeName=pelvis_capsule, center=<-4.1412 -3.2054 3.4851>, rotation=<rot 0 -0.0006 5>}
  {shapeType=Capsule, length=17.7796, radius=7.8001, shapeName=pelvis_capsule, center=<-6.628 1.4022 0>}
 ]
 thigh_l [{shapeType=Capsule, length=20.3362, radius=9.6638, shapeName=thigh_l_capsule, center=<-17.3111 -0.5401 0>, rotation=<rot -90 180 178.2614>}]
 thigh_r [{shapeType=Capsule, length=20.3362, radius=9.6638, shapeName=thigh_l_capsule, center=<17.3111 0.5401 0>, rotation=<rot -90 180 -1.7386>}]
}
constraints [
 {
  bone1Name=dfrm_front_strap_outer_04_l
  bone2Name=dfrm_front_strap_outer_03_l
  swing1LimitDegrees=25
  swing2LimitDegrees=10
  twistLimitDegrees=10
 }
 {
  bone1Name=dfrm_front_strap_outer_03_l
  bone2Name=dfrm_front_strap_outer_02_l
  swing1LimitDegrees=25
  swing2LimitDegrees=10
  twistLimitDegrees=10
 }
 {
  bone1Name=dfrm_front_strap_outer_02_l
  bone2Name=dfrm_front_strap_outer_01_l
  swing1LimitDegrees=25
  swing2LimitDegrees=10
  twistLimitDegrees=10
 }
 {
  bone1Name=dfrm_front_strap_outer_01_l
  bone2Name=dfrm_front_strap_outer_mstr_l
  swing1LimitDegrees=25
  swing2LimitDegrees=10
  twistLimitDegrees=10
 }
 {
  bone1Name=dfrm_front_strap_inner_04_l
  bone2Name=dfrm_front_strap_inner_03_l
  swing1LimitDegrees=25
  swing2LimitDegrees=10
  twistLimitDegrees=10
 }
 {
  bone1Name=dfrm_front_strap_inner_03_l
  bone2Name=dfrm_front_strap_inner_02_l
  swing1LimitDegrees=25
  swing2LimitDegrees=10
  twistLimitDegrees=10
 }
 {
  bone1Name=dfrm_front_strap_inner_02_l
  bone2Name=dfrm_front_strap_inner_01_l
  swing1LimitDegrees=25
  swing2LimitDegrees=10
  twistLimitDegrees=10
 }
 {
  bone1Name=dfrm_front_strap_inner_01_l
  bone2Name=dfrm_front_strap_inner_mstr_l
  swing1LimitDegrees=25
  swing2LimitDegrees=10
  twistLimitDegrees=10
 }
 {
  bone1Name=dfrm_front_strap_inner_04_r
  bone2Name=dfrm_front_strap_inner_03_r
  swing1LimitDegrees=25
  swing2LimitDegrees=10
  twistLimitDegrees=10
 }
 {
  bone1Name=dfrm_front_strap_outer_04_r
  bone2Name=dfrm_front_strap_outer_03_r
  swing1LimitDegrees=25
  swing2LimitDegrees=10
  twistLimitDegrees=10
 }
 {
  bone1Name=dfrm_front_strap_inner_03_r
  bone2Name=dfrm_front_strap_inner_02_r
  swing1LimitDegrees=25
  swing2LimitDegrees=10
  twistLimitDegrees=10
 }
 {
  bone1Name=dfrm_front_strap_outer_03_r
  bone2Name=dfrm_front_strap_outer_02_r
  swing1LimitDegrees=25
  swing2LimitDegrees=10
  twistLimitDegrees=10
 }
 {
  bone1Name=dfrm_front_strap_inner_02_r
  bone2Name=dfrm_front_strap_inner_01_r
  swing1LimitDegrees=25
  swing2LimitDegrees=10
  twistLimitDegrees=10
 }
 {
  bone1Name=dfrm_front_strap_outer_02_r
  bone2Name=dfrm_front_strap_outer_01_r
  swing1LimitDegrees=25
  swing2LimitDegrees=10
  twistLimitDegrees=10
 }
 {
  bone1Name=dfrm_front_strap_inner_01_r
  bone2Name=dfrm_front_strap_inner_mstr_r
  swing1LimitDegrees=25
  swing2LimitDegrees=10
  twistLimitDegrees=10
 }
 {
  bone1Name=dfrm_front_strap_outer_01_r
  bone2Name=dfrm_front_strap_outer_mstr_r
  swing1LimitDegrees=25
  swing2LimitDegrees=10
  twistLimitDegrees=10
 }
 {
  bone1Name=dfrm_back_strap_outer_01_l
  bone2Name=dfrm_back_strap_outer_mstr_l
  swing1LimitDegrees=35
  swing2LimitDegrees=10
  twistLimitDegrees=5
 }
 {
  bone1Name=dfrm_back_strap_inner_01_l
  bone2Name=dfrm_back_strap_inner_mstr_l
  swing1LimitDegrees=35
  swing2LimitDegrees=10
  twistLimitDegrees=5
 }
 {
  bone1Name=dfrm_back_strap_inner_01_r
  bone2Name=dfrm_back_strap_inner_mstr_r
  swing1LimitDegrees=35
  swing2LimitDegrees=10
  twistLimitDegrees=5
 }
 {
  bone1Name=dfrm_back_strap_outer_01_r
  bone2Name=dfrm_back_strap_outer_mstr_r
  swing1LimitDegrees=35
  swing2LimitDegrees=10
  twistLimitDegrees=5
 }
 {
  bone1Name=dfrm_back_strap_outer_02_l
  bone2Name=dfrm_back_strap_outer_01_l
  swing1LimitDegrees=20
  swing2LimitDegrees=15
  twistLimitDegrees=5
 }
 {
  bone1Name=dfrm_back_strap_inner_02_l
  bone2Name=dfrm_back_strap_inner_01_l
  swing1LimitDegrees=20
  swing2LimitDegrees=15
  twistLimitDegrees=5
 }
 {
  bone1Name=dfrm_back_strap_inner_02_r
  bone2Name=dfrm_back_strap_inner_01_r
  swing1LimitDegrees=20
  swing2LimitDegrees=15
  twistLimitDegrees=5
 }
 {
  bone1Name=dfrm_back_strap_outer_02_r
  bone2Name=dfrm_back_strap_outer_01_r
  swing1LimitDegrees=20
  swing2LimitDegrees=15
  twistLimitDegrees=5
 }
 {
  bone1Name=dfrm_back_strap_outer_03_l
  bone2Name=dfrm_back_strap_outer_02_l
  swing1LimitDegrees=20
  swing2LimitDegrees=15
  twistLimitDegrees=5
 }
 {
  bone1Name=dfrm_back_strap_inner_03_l
  bone2Name=dfrm_back_strap_inner_02_l
  swing1LimitDegrees=20
  swing2LimitDegrees=15
  twistLimitDegrees=5
 }
 {
  bone1Name=dfrm_back_strap_inner_03_r
  bone2Name=dfrm_back_strap_inner_02_r
  swing1LimitDegrees=20
  swing2LimitDegrees=15
  twistLimitDegrees=5
 }
 {
  bone1Name=dfrm_back_strap_outer_03_r
  bone2Name=dfrm_back_strap_outer_02_r
  swing1LimitDegrees=20
  swing2LimitDegrees=15
  twistLimitDegrees=5
 }
]
