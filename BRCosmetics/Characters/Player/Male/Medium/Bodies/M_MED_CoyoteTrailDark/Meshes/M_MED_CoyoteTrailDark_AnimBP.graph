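# Animation Blueprint for the M_MED_CoyoteTrailDark cosmetic character body, layering procedural rigs and physics over the parent mesh's pose.
# Copied pose passes through DeformRig, CtrlRig and a pelvis-space RigidBody simulation.
# Each update feeds movement alphas to CtrlRig, alongside crouch, jump and Stage2-variant helmet flags, and refreshes RigidBody external force via a deprecated gravity function.
# Spine and head twist calculations sit on unconnected chains.
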AnimBlueprint /BRCosmetics/Characters/Player/Male/Medium/Bodies/M_MED_CoyoteTrailDark/Meshes/M_MED_CoyoteTrailDark_AnimBP
target_skeleton=&/M_MED_CoyoteTrailDark_Skeleton
graphs [
 AnimGraph {nodes [
   n1 = Control Rig(Source=Copy Pose From Mesh() {bUseAttachedParent=true}) {
    alphaBoolBlend {blendOption=Linear}
    controlRigAssetReference {blueprintRigClass=&/M_MED_CoyoteTrailDark_DeformRig.M_MED_CoyoteTrailDark_DeformRig_C}
   }
   n2 = Control Rig!(
    Source=@n1
    Alpha
    is_helmet_active:bool=Get is_helmet_active() -> (is_helmet_active:bool)
    is_crouching:bool=Get bIsCrouching()
    is_jumping:bool=Get bIsJumping()
    left_right_movement_alpha:double=Get left_right_alpha() -> (left_right_alpha:double)
    forward_backwards_movement_alpha:double=Get forwards_backwards_alpha() -> (forwards_backwards_alpha:double)
   ) -> (Pose) {
    alphaBoolBlend {blendOption=Linear}
    controlRigAssetReference {blueprintRigClass=&/M_MED_CoyoteTrailDark_CtrlRig.M_MED_CoyoteTrailDark_CtrlRig_C}
   }
   n3 = RigidBody(ComponentPose=Local To Component(LocalPose=@n2), ExternalForce=Get gravity_override() -> (gravity_override:Vector)) {
    bClampLinearTranslationLimitToRefPose=true
    bOverrideWorldGravity=true
    overridePhysicsAsset=&/M_MED_CoyoteTrailDark_Physics
    simulationSpace=BaseBoneSpace
    alphaBoolBlend {blendOption=Linear}
    baseBoneRef {boneName=pelvis}
    componentAppliedLinearAccClamp=<3000 3000 3000>
    componentLinearAccScale=<0.4 0.3 0.3>
    componentLinearVelScale=<0.4 0.3 0.3>
    overrideWorldGravity {z=-980}
    simSpaceSettings {maxAngularAcceleration=1000, maxAngularVelocity=3, maxLinearAcceleration=250, maxLinearVelocity=500}
   }
   Output Pose(Result=Component To Local(ComponentPose=@n3)) {layerGroup=None}
  ]}
 EventGraph {nodes [
   Event BlueprintUpdateAnimation()
   > n1 = Sequence()
   > n2 = Gravity Override State Function DEPRECATED(CharacterPartAnimInstance=Self-Reference(), GameTimeInSeconds=GetGameTimeInSeconds())
   > Set gravity_override(gravity_override=@n2)
   n3 = PlayerMovement(execute=@n1.then_1, AnimInstance=Self-Reference(), XDotProductCoefficient=1, YDotProductCoefficient=1, ZDotProductCoefficient=1)
   > Set left_right_alpha(left_right_alpha:double=@n3) -> (Output_Get:double)
   > Set forwards_backwards_alpha(forwards_backwards_alpha:double=@n3.YMovement) -> (Output_Get:double)
   Event OnCharacterCustomizationLoaded()
   > n4 = Branch(Condition=CharacterItemDefinitionHasAnyMetaTag(Tags=Make Array("[0]"=(TagName="Cosmetics.Variant.Property.Stage2"))))
   > Set is_helmet_active(is_helmet_active:bool=true) -> (Output_Get:bool)
   Set is_helmet_active(execute=@n4.else, is_helmet_active=false)
   Event BlueprintInitializeAnimation()
   > Set Skeletal Mesh(Skeletal Mesh:Skeletal Mesh Component Object Reference=GetOwnerSkeletalMeshComponent()) -> (Output_Get:Skeletal Mesh Component Object Reference)
   Reroute Node()
   > Set Twist Left Alpha(Twist Left Alpha:double=GetAlphaFromReferenceTransformDelta(
     SkeletalMeshComponent=Get Skeletal Mesh() -> (Skeletal Mesh:Skeletal Mesh Component Object Reference)
     BoneName=spine_05
     BaseName=pelvis
     DeltaComponentType=QuaternionTwist
     InputRange=(X=-4,Y=8.5)
     TwistAxis="1,0,0"
    )) -> (Output_Get:double)
   Reroute Node()
   > Set spine_05_to_head_dRefz_alpha(spine_05_to_head_dRefz_alpha:double=GetAlphaFromReferenceTransformDelta(
     SkeletalMeshComponent=Get Skeletal Mesh()
     BoneName=head
     BaseName=spine_05
     DeltaComponentType=QuaternionTwist
     InputRange=(X=55,Y=91)
     TwistAxis="0,1,0"
    )) -> (Output_Get:double)
  ]}
]
variables [
 forwards_backwards_alpha {type=number}
 gravity_override {type=Vector}
 is_helmet_active {type=boolean}
 left_right_alpha {type=number}
 Skeletal Mesh {type=SkeletalMeshComponent}
 spine_05_to_head_dRefz_alpha {type=number}
 Twist Left Alpha {type=number}
]
